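# Physics asset for the arm mesh of the Moonlight Assassin medium male player cosmetic, setting up secondary motion for arm flaps and right-wrist tassels.
# Capsule bodies cover five two-bone dyn_ chains: top, mid and bottom flaps plus tassels A and B.
# Each chain is constrained bone to bone back to its own parent_dyn_ anchor body.
# Swing limits are wider on the flaps than on the tassels, with mostly narrow twist throughout.
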PhysicsAsset /BRCosmetics/Characters/Player/Male/Medium/Bodies/M_MED_MoonlightAssassin/Meshes/M_MED_MoonlightAssassin_Arm
bodies {
 dyn_armflap_bot_01 [{shapeType=Capsule, length=1.25, radius=1, center=<0.3471 0 0>, rotation=<rot 90 0 0>}]
 dyn_armflap_bot_02 [{shapeType=Capsule, length=4.75, radius=1.75, center=<2.0573 0 -0.4299>, rotation=<rot 90 0 0>}]
 dyn_armflap_mid02 [{shapeType=Capsule, length=0.75, radius=0.5, center=<0.7 0 0>, rotation=<rot 88.7941 175.927 155.4442>}]
 dyn_armflap_mid_01 [{shapeType=Capsule, length=1, radius=1.5, center=<0.3 0 0>, rotation=<rot 90 0 0>}]
 dyn_armflap_top_01 [{shapeType=Capsule, length=1.75, radius=1.75, center=<0.3 0 0>, rotation=<rot 90 0 0>}]
 dyn_armflap_top_02 [{shapeType=Capsule, length=2, radius=1.5, center=<0.7 0 0.5701>, rotation=<rot 90 0 0>}]
 dyn_wrist_tasselA_01_r [{shapeType=Capsule, length=0.75, radius=1.25, center=<0.3635 0 0>, rotation=<rot 90 0 0>}]
 dyn_wrist_tasselA_02_r [{shapeType=Capsule, length=1.25, radius=1, center=<1 0 0>, rotation=<rot 90 0 0>}]
 dyn_wrist_tasselB_01_r [{shapeType=Capsule, length=1.75, radius=1, center=<0.0551 0 0>, rotation=<rot 90 0 0>}]
 dyn_wrist_tasselB_02_r [{shapeType=Capsule, length=1.25, radius=1, center=<0.7 0 0>, rotation=<rot 90 0 0>}]
 parent_dyn_armflap_bot_01 [{shapeType=Capsule, length=0.505, radius=0.505, rotation=<rot 0 -90 0>}]
 parent_dyn_armflap_mid_01 [{shapeType=Capsule, length=0.505, radius=0.505, rotation=<rot 0 -90 0>}]
 parent_dyn_armflap_top_01 [{shapeType=Capsule, length=0.505, radius=0.505, rotation=<rot 0 -90 0>}]
 parent_dyn_wrist_tasselA_01_r [{shapeType=Capsule, length=0.505, radius=0.505, rotation=<rot 0 -90 0>}]
 parent_dyn_wrist_tasselB_01_r [{shapeType=Capsule, length=0.75, radius=1.25, center=<0.3 0 0>, rotation=<rot 90 0 0>}]
}
constraints [
 {bone1Name=dyn_armflap_top_01, bone2Name=parent_dyn_armflap_top_01, swing1LimitDegrees=45, swing2LimitDegrees=45, twistLimitDegrees=8}
 {bone1Name=dyn_armflap_top_02, bone2Name=dyn_armflap_top_01, swing1LimitDegrees=45, swing2LimitDegrees=45, twistLimitDegrees=8}
 {
  bone1Name=dyn_armflap_mid_01
  bone2Name=parent_dyn_armflap_mid_01
  swing1LimitDegrees=50.1429
  swing2LimitDegrees=57
  twistLimitDegrees=13.1429
 }
 {bone1Name=dyn_armflap_mid02, bone2Name=dyn_armflap_mid_01, swing1LimitDegrees=45, swing2LimitDegrees=45, twistLimitDegrees=8}
 {
  bone1Name=dyn_armflap_bot_01
  bone2Name=parent_dyn_armflap_bot_01
  swing1LimitDegrees=54.8572
  swing2LimitDegrees=43.2857
  twistLimitDegrees=8
 }
 {
  bone1Name=dyn_armflap_bot_02
  bone2Name=dyn_armflap_bot_01
  swing1LimitDegrees=60.4286
  swing2LimitDegrees=55.2857
  twistLimitDegrees=23.4286
 }
 {bone1Name=dyn_wrist_tasselA_01_r, bone2Name=parent_dyn_wrist_tasselA_01_r, swing1LimitDegrees=20, swing2LimitDegrees=25, twistLimitDegrees=8}
 {bone1Name=dyn_wrist_tasselA_02_r, bone2Name=dyn_wrist_tasselA_01_r, swing1LimitDegrees=20, swing2LimitDegrees=25, twistLimitDegrees=8}
 {bone1Name=dyn_wrist_tasselB_01_r, bone2Name=parent_dyn_wrist_tasselB_01_r, swing1LimitDegrees=20, swing2LimitDegrees=25, twistLimitDegrees=8}
 {bone1Name=dyn_wrist_tasselB_02_r, bone2Name=dyn_wrist_tasselB_01_r, swing1LimitDegrees=20, swing2LimitDegrees=25, twistLimitDegrees=8}
]
