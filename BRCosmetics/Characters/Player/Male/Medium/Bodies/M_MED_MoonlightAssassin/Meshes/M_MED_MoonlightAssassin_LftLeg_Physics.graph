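# Physics asset for the left-leg mesh of the medium male Moonlight Assassin body, with bodies for dynamic shorts and rear cloth bones.
# Pelvis, thigh and calf capsules cover the body; front, back and outer shorts chains hang from the thigh through parent, driver and main bones.
# A separate rearA chain attaches directly to the thigh.
# Root joints rotate freely on tiny or zero-size shapes, while flat-box tip segments have tighter twist limits.
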PhysicsAsset /BRCosmetics/Characters/Player/Male/Medium/Bodies/M_MED_MoonlightAssassin/Meshes/M_MED_MoonlightAssassin_LftLeg_Physics
bodies {
 calf_l [{shapeType=Capsule, length=25.8905, radius=6.3049, center=<-13.8306 -2.37 0.2503>, rotation=<rot -0.0019 -88.6287 -91.4137>}]
 driver_dyn_shorts_main_l [{shapeType=Capsule, length=1, radius=1}]
 dyn_rearA_01_l [{shapeType=Capsule, length=1, radius=1}]
 dyn_rearA_02_l [{shapeType=Capsule, length=0.5, radius=1.3313}]
 dyn_rearA_03_l [{shapeType=Box, extentX=1.0391, extentY=7.4555, extentZ=6.498, center=<4.002 0.0004 0>, rotation=<rot 0 90 0>}]
 dyn_rearA_04_l [{shapeType=Box, extentX=1.0286, extentY=8.8087, extentZ=5.6203, center=<4.3651 0.0004 0>, rotation=<rot 0 90 0>}]
 dyn_shorts_bkA_01_l [{shapeType=Capsule, length=1, radius=1}]
 dyn_shorts_bkA_02_l [{shapeType=Capsule, length=1, radius=1}]
 dyn_shorts_bkA_03_l [{shapeType=Box, extentX=1.0391, extentY=7.4555, extentZ=5.498, center=<4.002 0.0003 0.2772>, rotation=<rot 0 90 0>}]
 dyn_shorts_bkA_04_l [{shapeType=Box, extentX=1.0286, extentY=8.8087, extentZ=6.6203, center=<4.002 0.0004 0>, rotation=<rot 0 90 0>}]
 dyn_shorts_bkB_01_l [{shapeType=Capsule, length=1, radius=1}]
 dyn_shorts_bkB_02_l [{shapeType=Capsule, length=0.5, radius=1.3313}]
 dyn_shorts_bkB_03_l [{shapeType=Box, extentX=1.0391, extentY=7.4555, extentZ=6.498, center=<4.002 0.0004 0>, rotation=<rot 0 90 0>}]
 dyn_shorts_bkB_04_l [{shapeType=Box, extentX=1.0286, extentY=8.8087, extentZ=5.6203, center=<4.3651 0.0004 0>, rotation=<rot 0 90 0>}]
 dyn_shorts_frt_01_l [{shapeType=Box, extentX=0, extentY=0, extentZ=0}]
 dyn_shorts_frt_02_l [{shapeType=Box, extentX=0, extentY=0, extentZ=0}]
 dyn_shorts_frt_03_l [{shapeType=Box, extentX=1.0391, extentY=7.4555, extentZ=5.498, center=<4.002 0.0003 0.2772>, rotation=<rot 0 90 0>}]
 dyn_shorts_frt_04_l [{shapeType=Box, extentX=1.0286, extentY=8.8087, extentZ=6.6203, center=<4.002 0.0004 0>, rotation=<rot 0 90 0>}]
 dyn_shorts_main_l [{shapeType=Box, extentX=0, extentY=0, extentZ=0}]
 dyn_shorts_main_rear_l [{shapeType=Capsule, length=1, radius=1}]
 dyn_shorts_outA_01_l [{shapeType=Box, extentX=0, extentY=0, extentZ=0}]
 dyn_shorts_outA_02_l [{shapeType=Box, extentX=0, extentY=0, extentZ=0}]
 dyn_shorts_outA_03_l [{shapeType=Box, extentX=1.0391, extentY=7.4555, extentZ=5.498, center=<4.002 0.0003 0.2772>, rotation=<rot 0 90 0>}]
 dyn_shorts_outA_04_l [{shapeType=Box, extentX=1.0286, extentY=8.8087, extentZ=6.6203, center=<4.002 0.0004 0>, rotation=<rot 0 90 0>}]
 dyn_shorts_outB_01_l [{shapeType=Box, extentX=0, extentY=0, extentZ=0}]
 dyn_shorts_outB_02_l [{shapeType=Box, extentX=0, extentY=0, extentZ=0}]
 dyn_shorts_outB_03_l [{shapeType=Box, extentX=1.0391, extentY=7.4555, extentZ=5.498, center=<4.002 0.0003 0.2772>, rotation=<rot 0 90 0>}]
 dyn_shorts_outB_04_l [{shapeType=Box, extentX=1.0286, extentY=8.8087, extentZ=6.6203, center=<4.002 0.0004 0>, rotation=<rot 0 90 0>}]
 parent_dyn_shorts_main_l [{shapeType=Capsule, length=1, radius=1}]
 pelvis [{shapeType=Capsule, length=19.5626, radius=5}]
 thigh_twist_01_l [{shapeType=Capsule, length=24.415, radius=6.6335, center=<0.4567 1.2991 0.0186>, rotation=<rot -88.3234 6.0423 179.9998>}]
}
constraints [
 {
  bone1Name=dyn_shorts_bkB_02_l
  bone2Name=dyn_shorts_bkB_01_l
  swing1LimitDegrees=45
  swing1Motion=Free
  swing2LimitDegrees=45
  swing2Motion=Free
  twistLimitDegrees=45
  twistMotion=Free
 }
 {
  bone1Name=dyn_shorts_bkB_01_l
  bone2Name=dyn_shorts_main_rear_l
  swing1LimitDegrees=45
  swing1Motion=Free
  swing2LimitDegrees=45
  swing2Motion=Free
  twistLimitDegrees=45
  twistMotion=Free
 }
 {
  bone1Name=dyn_shorts_main_rear_l
  bone2Name=driver_dyn_shorts_main_l
  swing1LimitDegrees=45
  swing1Motion=Free
  swing2LimitDegrees=45
  swing2Motion=Free
  twistLimitDegrees=45
  twistMotion=Free
 }
 {
  bone1Name=driver_dyn_shorts_main_l
  bone2Name=parent_dyn_shorts_main_l
  swing1LimitDegrees=45
  swing1Motion=Free
  swing2LimitDegrees=45
  swing2Motion=Free
  twistLimitDegrees=45
  twistMotion=Free
 }
 {
  bone1Name=parent_dyn_shorts_main_l
  bone2Name=thigh_twist_01_l
  swing1LimitDegrees=45
  swing1Motion=Free
  swing2LimitDegrees=45
  swing2Motion=Free
  twistLimitDegrees=45
  twistMotion=Free
 }
 {bone1Name=dyn_shorts_bkB_03_l, bone2Name=dyn_shorts_bkB_02_l, swing1LimitDegrees=45, swing2LimitDegrees=45, twistLimitDegrees=15}
 {bone1Name=dyn_shorts_bkB_04_l, bone2Name=dyn_shorts_bkB_03_l, swing1LimitDegrees=45, swing2LimitDegrees=45, twistLimitDegrees=15}
 {
  bone1Name=calf_l
  bone2Name=pelvis
  swing1LimitDegrees=45
  swing1Motion=Free
  swing2LimitDegrees=45
  swing2Motion=Free
  twistLimitDegrees=45
  twistMotion=Free
 }
 {
  bone1Name=thigh_twist_01_l
  bone2Name=pelvis
  swing1LimitDegrees=45
  swing1Motion=Free
  swing2LimitDegrees=45
  swing2Motion=Free
  twistLimitDegrees=45
  twistMotion=Free
 }
 {
  bone1Name=dyn_shorts_bkA_01_l
  bone2Name=dyn_shorts_main_rear_l
  swing1LimitDegrees=45
  swing1Motion=Free
  swing2LimitDegrees=45
  swing2Motion=Free
  twistLimitDegrees=45
  twistMotion=Free
 }
 {
  bone1Name=dyn_shorts_bkA_02_l
  bone2Name=dyn_shorts_bkA_01_l
  swing1LimitDegrees=45
  swing1Motion=Free
  swing2LimitDegrees=45
  swing2Motion=Free
  twistLimitDegrees=45
  twistMotion=Free
 }
 {bone1Name=dyn_shorts_bkA_03_l, bone2Name=dyn_shorts_bkA_02_l, swing1LimitDegrees=45, swing2LimitDegrees=45, twistLimitDegrees=15}
 {bone1Name=dyn_shorts_bkA_04_l, bone2Name=dyn_shorts_bkA_03_l, swing1LimitDegrees=45, swing2LimitDegrees=45, twistLimitDegrees=15}
 {bone1Name=dyn_shorts_outB_03_l, bone2Name=dyn_shorts_outB_02_l, swing1LimitDegrees=45, swing2LimitDegrees=45, twistLimitDegrees=15}
 {bone1Name=dyn_shorts_outB_04_l, bone2Name=dyn_shorts_outB_03_l, swing1LimitDegrees=45, swing2LimitDegrees=45, twistLimitDegrees=15}
 {
  bone1Name=dyn_shorts_outB_02_l
  bone2Name=dyn_shorts_outB_01_l
  swing1LimitDegrees=45
  swing1Motion=Free
  swing2LimitDegrees=45
  swing2Motion=Free
  twistLimitDegrees=45
  twistMotion=Free
 }
 {
  bone1Name=dyn_shorts_outB_01_l
  bone2Name=dyn_shorts_main_rear_l
  swing1LimitDegrees=45
  swing1Motion=Free
  swing2LimitDegrees=45
  swing2Motion=Free
  twistLimitDegrees=45
  twistMotion=Free
 }
 {
  bone1Name=dyn_shorts_outA_01_l
  bone2Name=dyn_shorts_main_rear_l
  swing1LimitDegrees=45
  swing1Motion=Free
  swing2LimitDegrees=45
  swing2Motion=Free
  twistLimitDegrees=45
  twistMotion=Free
 }
 {
  bone1Name=dyn_shorts_outA_02_l
  bone2Name=dyn_shorts_outA_01_l
  swing1LimitDegrees=45
  swing1Motion=Free
  swing2LimitDegrees=45
  swing2Motion=Free
  twistLimitDegrees=45
  twistMotion=Free
 }
 {bone1Name=dyn_shorts_outA_03_l, bone2Name=dyn_shorts_outA_02_l, swing1LimitDegrees=45, swing2LimitDegrees=45, twistLimitDegrees=15}
 {bone1Name=dyn_shorts_outA_04_l, bone2Name=dyn_shorts_outA_03_l, swing1LimitDegrees=45, swing2LimitDegrees=45, twistLimitDegrees=15}
 {
  bone1Name=dyn_shorts_main_l
  bone2Name=driver_dyn_shorts_main_l
  swing1LimitDegrees=45
  swing1Motion=Free
  swing2LimitDegrees=45
  swing2Motion=Free
  twistLimitDegrees=45
  twistMotion=Free
 }
 {
  bone1Name=dyn_shorts_frt_01_l
  bone2Name=dyn_shorts_main_l
  swing1LimitDegrees=45
  swing1Motion=Free
  swing2LimitDegrees=45
  swing2Motion=Free
  twistLimitDegrees=45
  twistMotion=Free
 }
 {
  bone1Name=dyn_shorts_frt_02_l
  bone2Name=dyn_shorts_frt_01_l
  swing1LimitDegrees=45
  swing1Motion=Free
  swing2LimitDegrees=45
  swing2Motion=Free
  twistLimitDegrees=45
  twistMotion=Free
 }
 {bone1Name=dyn_shorts_frt_03_l, bone2Name=dyn_shorts_frt_02_l, swing1LimitDegrees=45, swing2LimitDegrees=45, twistLimitDegrees=15}
 {bone1Name=dyn_shorts_frt_04_l, bone2Name=dyn_shorts_frt_03_l, swing1LimitDegrees=45, swing2LimitDegrees=45, twistLimitDegrees=15}
 {
  bone1Name=dyn_rearA_01_l
  bone2Name=thigh_twist_01_l
  swing1LimitDegrees=45
  swing1Motion=Free
  swing2LimitDegrees=45
  swing2Motion=Free
  twistLimitDegrees=45
  twistMotion=Free
 }
 {
  bone1Name=dyn_rearA_02_l
  bone2Name=dyn_rearA_01_l
  swing1LimitDegrees=45
  swing1Motion=Free
  swing2LimitDegrees=45
  swing2Motion=Free
  twistLimitDegrees=45
  twistMotion=Free
 }
 {bone1Name=dyn_rearA_03_l, bone2Name=dyn_rearA_02_l, swing1LimitDegrees=45, swing2LimitDegrees=45, twistLimitDegrees=15}
 {bone1Name=dyn_rearA_04_l, bone2Name=dyn_rearA_03_l, swing1LimitDegrees=45, swing2LimitDegrees=45, twistLimitDegrees=15}
]
