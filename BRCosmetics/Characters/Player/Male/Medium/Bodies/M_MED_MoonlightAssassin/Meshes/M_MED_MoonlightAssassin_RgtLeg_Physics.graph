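# Physics asset for the Moonlight Assassin (male, medium) right-leg mesh, providing bodies and constraints for its dynamic shorts and rear cloth bones.
# Capsules cover the pelvis and both legs as body volumes, while tiny capsules mark chain roots and thin boxes form the outer cloth segments.
# Shorts panels branch from the right thigh through parent, driver and main bones; rearA attaches directly.
# Root joints rotate freely, while outer segments have angular limits with tighter twist.
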PhysicsAsset /BRCosmetics/Characters/Player/Male/Medium/Bodies/M_MED_MoonlightAssassin/Meshes/M_MED_MoonlightAssassin_RgtLeg_Physics
bodies {
 calf_l [{shapeType=Capsule, length=25.8905, radius=6.3049, center=<-13.8306 -2.37 0.2503>, rotation=<rot -0.0019 -88.6287 -91.4137>}]
 calf_r [{shapeType=Capsule, length=25.3905, radius=6.8049, center=<13.2773 3.381 0.3538>, rotation=<rot -0.0443 -90.3456 -89.2593>}]
 driver_dyn_shorts_main_r [{shapeType=Capsule, length=1, radius=1}]
 dyn_rearA_01_r [{shapeType=Capsule, length=1, radius=1}]
 dyn_rearA_02_r [{shapeType=Capsule, length=1, radius=1}]
 dyn_rearA_03_r [{shapeType=Box, extentX=8, extentY=0.5, extentZ=7.5, center=<-3.878 0 0>}]
 dyn_rearA_04_r [{shapeType=Box, extentX=9, extentY=0.5, extentZ=8, center=<-4.3504 0.0001 0>}]
 dyn_shorts_bkA_01_r [{shapeType=Capsule, length=1, radius=1}]
 dyn_shorts_bkA_02_r [{shapeType=Capsule, length=1, radius=1}]
 dyn_shorts_bkA_03_r [{shapeType=Box, extentX=8, extentY=0.5, extentZ=7.5, center=<-3.878 0 0>}]
 dyn_shorts_bkA_04_r [{shapeType=Box, extentX=9, extentY=0.5, extentZ=8, center=<-4.3504 0.0001 0>}]
 dyn_shorts_bkB_01_r [{shapeType=Capsule, length=1, radius=1}]
 dyn_shorts_bkB_02_r [{shapeType=Capsule, length=1, radius=1}]
 dyn_shorts_bkB_03_r [{shapeType=Box, extentX=8, extentY=0.5, extentZ=7.5, center=<-3.878 0 0>}]
 dyn_shorts_bkB_04_r [{shapeType=Box, extentX=9, extentY=0.5, extentZ=8, center=<-4.3504 0.0001 0>}]
 dyn_shorts_frtA_01_r [{shapeType=Capsule, length=1, radius=1}]
 dyn_shorts_frtA_02_r [{shapeType=Capsule, length=1, radius=1}]
 dyn_shorts_frtA_03_r [{shapeType=Box, extentX=1.0391, extentY=9.4555, extentZ=6.498, center=<4.2261 0.0004 0.2772>, rotation=<rot 0 90 0>}]
 dyn_shorts_frtA_04_r [{shapeType=Box, extentX=1.0391, extentY=9.4555, extentZ=6.498, center=<4.2261 0.0004 0.2772>, rotation=<rot 0 90 0>}]
 dyn_shorts_frtA_05_r [{shapeType=Box, extentX=1.0286, extentY=8.8087, extentZ=6.6203, center=<4.002 0.0004 0>, rotation=<rot 0 90 0>}]
 dyn_shorts_frt_01_r [{shapeType=Capsule, length=1, radius=1}]
 dyn_shorts_frt_02_r [{shapeType=Capsule, length=1, radius=1}]
 dyn_shorts_frt_03_r [{shapeType=Box, extentX=11, extentY=0.5, extentZ=7.5, center=<-5.0813 0 0>}]
 dyn_shorts_frt_04_r [{shapeType=Box, extentX=9, extentY=0.5, extentZ=6.5, center=<-4.3504 0.0001 0>}]
 dyn_shorts_main_r [{shapeType=Capsule, length=1, radius=1}]
 dyn_shorts_main_rear_r [{shapeType=Capsule, length=1, radius=1}]
 dyn_shorts_outA_01_r [{shapeType=Capsule, length=1, radius=1}]
 dyn_shorts_outA_02_r [{shapeType=Capsule, length=1, radius=1}]
 dyn_shorts_outA_03_r [{shapeType=Box, extentX=8, extentY=0.5, extentZ=7.5, center=<-3.878 0 0>}]
 dyn_shorts_outA_04_r [{shapeType=Box, extentX=10.5, extentY=0.5, extentZ=6, center=<-4.3504 0.0001 0>}]
 dyn_shorts_outB_01_r [{shapeType=Capsule, length=1, radius=1}]
 dyn_shorts_outB_02_r [{shapeType=Capsule, length=1, radius=1}]
 dyn_shorts_outB_03_r [{shapeType=Box, extentX=8, extentY=0.5, extentZ=7.5, center=<-3.878 0 0>}]
 dyn_shorts_outB_04_r [{shapeType=Box, extentX=9, extentY=0.5, extentZ=7, center=<-4.3504 0.0001 0>}]
 parent_dyn_shorts_main_r [{shapeType=Capsule, length=1, radius=1}]
 pelvis [{shapeType=Capsule, length=19.5626, radius=5}]
 thigh_r [{shapeType=Capsule, length=24.415, radius=6.6335, center=<21.4484 3.5215 -0.1829>, rotation=<rot -88.3234 6.0423 179.9998>}]
 thigh_twist_01_l [{shapeType=Capsule, length=24.415, radius=6.6335, center=<0.4567 1.2991 0.0186>, rotation=<rot -88.3234 6.0423 179.9998>}]
}
constraints [
 {
  bone1Name=calf_l
  bone2Name=pelvis
  swing1LimitDegrees=45
  swing1Motion=Free
  swing2LimitDegrees=45
  swing2Motion=Free
  twistLimitDegrees=45
  twistMotion=Free
 }
 {
  bone1Name=thigh_twist_01_l
  bone2Name=pelvis
  swing1LimitDegrees=45
  swing1Motion=Free
  swing2LimitDegrees=45
  swing2Motion=Free
  twistLimitDegrees=45
  twistMotion=Free
 }
 {
  bone1Name=thigh_r
  bone2Name=pelvis
  swing1LimitDegrees=45
  swing1Motion=Free
  swing2LimitDegrees=45
  swing2Motion=Free
  twistLimitDegrees=45
  twistMotion=Free
 }
 {
  bone1Name=calf_r
  bone2Name=thigh_r
  swing1LimitDegrees=45
  swing1Motion=Free
  swing2LimitDegrees=45
  swing2Motion=Free
  twistLimitDegrees=45
  twistMotion=Free
 }
 {
  bone1Name=dyn_shorts_main_rear_r
  bone2Name=driver_dyn_shorts_main_r
  swing1LimitDegrees=45
  swing1Motion=Free
  swing2LimitDegrees=45
  swing2Motion=Free
  twistLimitDegrees=45
  twistMotion=Free
 }
 {
  bone1Name=driver_dyn_shorts_main_r
  bone2Name=parent_dyn_shorts_main_r
  swing1LimitDegrees=45
  swing1Motion=Free
  swing2LimitDegrees=45
  swing2Motion=Free
  twistLimitDegrees=45
  twistMotion=Free
 }
 {
  bone1Name=parent_dyn_shorts_main_r
  bone2Name=thigh_r
  swing1LimitDegrees=45
  swing1Motion=Free
  swing2LimitDegrees=45
  swing2Motion=Free
  twistLimitDegrees=45
  twistMotion=Free
 }
 {
  bone1Name=dyn_shorts_outA_01_r
  bone2Name=dyn_shorts_main_rear_r
  swing1LimitDegrees=45
  swing1Motion=Free
  swing2LimitDegrees=45
  swing2Motion=Free
  twistLimitDegrees=45
  twistMotion=Free
 }
 {
  bone1Name=dyn_shorts_outA_02_r
  bone2Name=dyn_shorts_outA_01_r
  swing1LimitDegrees=45
  swing1Motion=Free
  swing2LimitDegrees=45
  swing2Motion=Free
  twistLimitDegrees=45
  twistMotion=Free
 }
 {
  bone1Name=dyn_shorts_outB_01_r
  bone2Name=dyn_shorts_main_rear_r
  swing1LimitDegrees=45
  swing1Motion=Free
  swing2LimitDegrees=45
  swing2Motion=Free
  twistLimitDegrees=45
  twistMotion=Free
 }
 {
  bone1Name=dyn_shorts_outB_02_r
  bone2Name=dyn_shorts_outB_01_r
  swing1LimitDegrees=45
  swing1Motion=Free
  swing2LimitDegrees=45
  swing2Motion=Free
  twistLimitDegrees=45
  twistMotion=Free
 }
 {
  bone1Name=dyn_shorts_bkA_01_r
  bone2Name=dyn_shorts_main_rear_r
  swing1LimitDegrees=45
  swing1Motion=Free
  swing2LimitDegrees=45
  swing2Motion=Free
  twistLimitDegrees=45
  twistMotion=Free
 }
 {
  bone1Name=dyn_shorts_bkA_02_r
  bone2Name=dyn_shorts_bkA_01_r
  swing1LimitDegrees=45
  swing1Motion=Free
  swing2LimitDegrees=45
  swing2Motion=Free
  twistLimitDegrees=45
  twistMotion=Free
 }
 {
  bone1Name=dyn_shorts_bkB_01_r
  bone2Name=dyn_shorts_main_rear_r
  swing1LimitDegrees=45
  swing1Motion=Free
  swing2LimitDegrees=45
  swing2Motion=Free
  twistLimitDegrees=45
  twistMotion=Free
 }
 {
  bone1Name=dyn_shorts_bkB_02_r
  bone2Name=dyn_shorts_bkB_01_r
  swing1LimitDegrees=45
  swing1Motion=Free
  swing2LimitDegrees=45
  swing2Motion=Free
  twistLimitDegrees=45
  twistMotion=Free
 }
 {
  bone1Name=dyn_shorts_main_r
  bone2Name=driver_dyn_shorts_main_r
  swing1LimitDegrees=45
  swing1Motion=Free
  swing2LimitDegrees=45
  swing2Motion=Free
  twistLimitDegrees=45
  twistMotion=Free
 }
 {
  bone1Name=dyn_shorts_frt_01_r
  bone2Name=dyn_shorts_main_r
  swing1LimitDegrees=45
  swing1Motion=Free
  swing2LimitDegrees=45
  swing2Motion=Free
  twistLimitDegrees=45
  twistMotion=Free
 }
 {
  bone1Name=dyn_shorts_frt_02_r
  bone2Name=dyn_shorts_frt_01_r
  swing1LimitDegrees=45
  swing1Motion=Free
  swing2LimitDegrees=45
  swing2Motion=Free
  twistLimitDegrees=45
  twistMotion=Free
 }
 {
  bone1Name=dyn_shorts_frtA_01_r
  bone2Name=dyn_shorts_main_r
  swing1LimitDegrees=45
  swing1Motion=Free
  swing2LimitDegrees=45
  swing2Motion=Free
  twistLimitDegrees=45
  twistMotion=Free
 }
 {
  bone1Name=dyn_shorts_frtA_02_r
  bone2Name=dyn_shorts_frtA_01_r
  swing1LimitDegrees=45
  swing1Motion=Free
  swing2LimitDegrees=45
  swing2Motion=Free
  twistLimitDegrees=45
  twistMotion=Free
 }
 {
  bone1Name=dyn_rearA_01_r
  bone2Name=thigh_r
  swing1LimitDegrees=45
  swing1Motion=Free
  swing2LimitDegrees=45
  swing2Motion=Free
  twistLimitDegrees=45
  twistMotion=Free
 }
 {
  bone1Name=dyn_rearA_02_r
  bone2Name=dyn_rearA_01_r
  swing1LimitDegrees=45
  swing1Motion=Free
  swing2LimitDegrees=45
  swing2Motion=Free
  twistLimitDegrees=45
  twistMotion=Free
 }
 {bone1Name=dyn_shorts_outA_03_r, bone2Name=dyn_shorts_outA_02_r, swing1LimitDegrees=45, swing2LimitDegrees=45, twistLimitDegrees=15}
 {bone1Name=dyn_shorts_outA_04_r, bone2Name=dyn_shorts_outA_03_r, swing1LimitDegrees=45, swing2LimitDegrees=45, twistLimitDegrees=15}
 {bone1Name=dyn_shorts_outB_03_r, bone2Name=dyn_shorts_outB_02_r, swing1LimitDegrees=45, swing2LimitDegrees=45, twistLimitDegrees=15}
 {bone1Name=dyn_shorts_outB_04_r, bone2Name=dyn_shorts_outB_03_r, swing1LimitDegrees=45, swing2LimitDegrees=45, twistLimitDegrees=15}
 {bone1Name=dyn_shorts_bkA_03_r, bone2Name=dyn_shorts_bkA_02_r, swing1LimitDegrees=45, swing2LimitDegrees=45, twistLimitDegrees=15}
 {bone1Name=dyn_shorts_bkA_04_r, bone2Name=dyn_shorts_bkA_03_r, swing1LimitDegrees=45, swing2LimitDegrees=45, twistLimitDegrees=15}
 {bone1Name=dyn_shorts_bkB_03_r, bone2Name=dyn_shorts_bkB_02_r, swing1LimitDegrees=45, swing2LimitDegrees=45, twistLimitDegrees=15}
 {bone1Name=dyn_shorts_bkB_04_r, bone2Name=dyn_shorts_bkB_03_r, swing1LimitDegrees=45, swing2LimitDegrees=45, twistLimitDegrees=15}
 {bone1Name=dyn_shorts_frt_03_r, bone2Name=dyn_shorts_frt_02_r, swing1LimitDegrees=45, swing2LimitDegrees=45, twistLimitDegrees=15}
 {bone1Name=dyn_shorts_frt_04_r, bone2Name=dyn_shorts_frt_03_r, swing1LimitDegrees=45, swing2LimitDegrees=45, twistLimitDegrees=15}
 {bone1Name=dyn_shorts_frtA_03_r, bone2Name=dyn_shorts_frtA_02_r, swing1LimitDegrees=45, swing2LimitDegrees=45, twistLimitDegrees=15}
 {bone1Name=dyn_shorts_frtA_04_r, bone2Name=dyn_shorts_frtA_03_r, swing1LimitDegrees=45, swing2LimitDegrees=45, twistLimitDegrees=15}
 {bone1Name=dyn_shorts_frtA_05_r, bone2Name=dyn_shorts_frtA_04_r, swing1LimitDegrees=45, swing2LimitDegrees=45, twistLimitDegrees=15}
 {bone1Name=dyn_rearA_03_r, bone2Name=dyn_rearA_02_r, swing1LimitDegrees=45, swing2LimitDegrees=45, twistLimitDegrees=15}
 {bone1Name=dyn_rearA_04_r, bone2Name=dyn_rearA_03_r, swing1LimitDegrees=45, swing2LimitDegrees=45, twistLimitDegrees=15}
]
