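# Physics asset of simulation colliders for the Moonlight Assassin medium male player body cosmetic, covering only the lower body.
# Five capsules approximate the pelvis, both thighs and both calves, with slightly different sizes on left and right.
# Constraints chain each calf to its thigh and each thigh to the pelvis, with swing and twist left free.
# The name suggests these shapes act as colliders for simulated outfit parts rather than as a ragdoll.
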PhysicsAsset /BRCosmetics/Characters/Player/Male/Medium/Bodies/M_MED_MoonlightAssassin/Meshes/M_MED_MoonlightAssassin_Sim_Colliders
bodies {
 calf_l [{shapeType=Capsule, length=35.242, radius=7.9116, center=<-19.137 -1.7389 -0.2269>, rotation=<rot 0.0342 90.4044 -89.4664>}]
 calf_r [{shapeType=Capsule, length=35.5362, radius=7.5971, center=<19.4572 1.9586 0.2419>, rotation=<rot 0.0249 91.3556 -89.3731>}]
 pelvis [{shapeType=Capsule, length=11.4559, radius=9.0829, center=<2.2747 -0.2766 1.0154>}]
 thigh_l [{shapeType=Capsule, length=29.252, radius=7.3318, center=<-19.839 -0.0002 0>, rotation=<rot 90 0 0>}]
 thigh_r [{shapeType=Capsule, length=29.8653, radius=6.1815, center=<19.0128 0 0>, rotation=<rot 90 0 0>}]
}
constraints [
 {
  bone1Name=thigh_l
  bone2Name=pelvis
  swing1LimitDegrees=45
  swing1Motion=Free
  swing2LimitDegrees=45
  swing2Motion=Free
  twistLimitDegrees=45
  twistMotion=Free
 }
 {
  bone1Name=calf_l
  bone2Name=thigh_l
  swing1LimitDegrees=45
  swing1Motion=Free
  swing2LimitDegrees=45
  swing2Motion=Free
  twistLimitDegrees=45
  twistMotion=Free
 }
 {
  bone1Name=thigh_r
  bone2Name=pelvis
  swing1LimitDegrees=45
  swing1Motion=Free
  swing2LimitDegrees=45
  swing2Motion=Free
  twistLimitDegrees=45
  twistMotion=Free
 }
 {
  bone1Name=calf_r
  bone2Name=thigh_r
  swing1LimitDegrees=45
  swing1Motion=Free
  swing2LimitDegrees=45
  swing2Motion=Free
  twistLimitDegrees=45
  twistMotion=Free
 }
]
